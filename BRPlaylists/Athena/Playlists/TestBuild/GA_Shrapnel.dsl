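; Test-build gameplay ability that makes a player-built structure throw an explosion out its far side when destroyed.
; Activation from a gameplay event binds OnDied on the target to a stair, wall or cone/floor handler.
; Handlers use a dot-product angle test against the damage causer to set an offset TargetTransform, returned as the custom ability source transform.
; Target selection then applies Explode effect containers, alongside a gameplay cue and firework particles.
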(event Custom|Play%20Client%20Effects
  (params (Location :type Vector))
  (Effects|SpawnEmitteratLocation "/Game/Athena/Items/Consumables/FireworksMortar/P_Athena_FireworksMortar_Explosion.P_Athena_FireworksMortar_Explosion" Location))

(event Ability|EventActivateAbilityFromEvent
  (params (EventData :type Gameplay%20Event%20Data%20Structure%20%28by%20ref%29))
  (bind (_eventtag _instigator _target) (Utilities|Struct|BreakGameplayEventData EventData))
  (bind (_asbuilding_wall _bsuccess) (Utilities|Casting|CastToBuildingWall (Utilities|Casting|CastToBuildingSMActor _target)))
  (bind (_asbuilding_static_mesh _bsuccess_2) (Utilities|Casting|CastToBuildingSMActor _target))
  (bind (_asbuilding_stairs _bsuccess_3) (Utilities|Casting|CastToBuildingStairs _asbuilding_static_mesh))
  (if _bsuccess_2
    (if (BuildingActor|IsPlayerBuilt _asbuilding_static_mesh)
      (Development|PrintString "Is player built")
      (if _bsuccess_3
        (Development|PrintString "Stairs")
        (EventDispatchers|BindEventtoOnDied _asbuilding_static_mesh (EventDispatchers|CreateEvent :!event Custom|Stair))
        (elif _bsuccess
          (Development|PrintString "Wall")
          (EventDispatchers|BindEventtoOnDied _asbuilding_static_mesh (EventDispatchers|CreateEvent :!event Custom|Wall/Floor/Roof))
          (else
            (Development|PrintString "Cone/Floor")
            (EventDispatchers|BindEventtoOnDied _asbuilding_static_mesh (EventDispatchers|CreateEvent :!event Custom|Cone/Floor)))))
      (else
        (Development|PrintString "is not player built")))))

(event Custom|Wall/Floor/Roof
  (params (DamagedActor :type Actor%20Object%20Reference) (Damage :type Float%20%28single-precision%29) (InstigatedBy :type Controller%20Object%20Reference) (DamageCauser :type Actor%20Object%20Reference) (HitLocation :type Vector) (FHitComponent :type Primitive%20Component%20Object%20Reference) (BoneName :type Name) (Momentum :type Vector))
  (bind _returnvalue (Transformation|GetActorLocation DamagedActor))
  (bind _targettransform (Variables|Default|GetTargetTransform))
  (bind _location (.location _targettransform))
  (Variables|Default|SetAttackingPlayer InstigatedBy)
  (bind _output_get (Variables|Default|SetRightVector (Transformation|GetActorRightVector DamagedActor)))
  (Variables|Default|SetActor%20Location _returnvalue)
  (Development|PrintString "Spawned Actor")
  ;[1 1 1] if greater than. Use forward Vector, if less than, use inverse, then decide where to spawn the actor
  (if (> (Math|Float|Absolute(Float) (Math|Trig|Acos(Degrees) (Math|Vector|DotProduct _output_get (Math|Vector|GetUnitDirection(Vector) _returnvalue (Transformation|GetActorLocation DamageCauser))))) 9e+01)
    (Development|PrintString "Backward Vector")
    (Variables|Default|SetTargetTransform (Math|Transform|MakeTransform (Math|Vector|MakeVector (.x (+ (Variables|Default|GetActor%20Location) (* (Variables|Default|GetRightVector) 256.0))) (.y (+ (Variables|Default|GetActor%20Location) (* (Variables|Default|GetRightVector) 256.0))) (+ (.z (+ (Variables|Default|GetActor%20Location) (* (Variables|Default|GetRightVector) 256.0))) 256.0)) (Math|Conversions|RotationFromXVector (Variables|Default|GetRightVector))))
    (bind (_asynctaskproxy _targetdata _applicationtag) (Ability|Tasks|DoTargetSelection "None" "(TagName=\"Gameplay.Application.Explode\")" true)
      (:then
        (Ability|ExecuteGameplayCueWithParamsOnOwner (Variables|Default|GetGameplayCueTag) (Ability|GameplayCue|MakeGameplayCueParameters :NormalizedMagnitude 0.0 :RawMagnitude 0.0 :Location _location :Normal "0,0,0" :Instigator (Utilities|Casting|CastToFortPlayerPawn (Variables|Default|GetAttackingPlayer)) :GameplayEffectLevel 1 :AbilityLevel 1 :bReplicateLocationWhenUsingMinimalRepProxy false)))
      (:Targeted
        (Ability|ApplyGameplayEffectContainer _targetdata _applicationtag)))
    (else
      (Development|PrintString "Forward Vector")
      (Variables|Default|SetTargetTransform (Math|Transform|MakeTransform (Math|Vector|MakeVector (.x (- (Variables|Default|GetActor%20Location) (* (Variables|Default|GetRightVector) 256.0))) (.y (- (Variables|Default|GetActor%20Location) (* (Variables|Default|GetRightVector) 256.0))) (+ (.z (- (Variables|Default|GetActor%20Location) (* (Variables|Default|GetRightVector) 256.0))) 256.0)) (Math|Conversions|RotationFromXVector (Math|Transform|InverseTransformDirection (Math|Conversions|ToTransform(Vector) (Variables|Default|GetRightVector)) (Variables|Default|GetRightVector)))))
      (bind (_asynctaskproxy_2 _targetdata_2 _applicationtag_2) (Ability|Tasks|DoTargetSelection "None" "(TagName=\"Gameplay.Application.Explode\")" true)
        (:then
          (Ability|ExecuteGameplayCueWithParamsOnOwner (Variables|Default|GetGameplayCueTag) (Ability|GameplayCue|MakeGameplayCueParameters :NormalizedMagnitude 0.0 :RawMagnitude 0.0 :Location _location :Normal "0,0,0" :Instigator (Utilities|Casting|CastToFortPlayerPawn (Variables|Default|GetAttackingPlayer)) :GameplayEffectLevel 1 :AbilityLevel 1 :bReplicateLocationWhenUsingMinimalRepProxy false)))
        (:Targeted
          (Ability|ApplyGameplayEffectContainer _targetdata_2 _applicationtag_2)))))
  (CallFunction|PlayClientEffects _location))

(event Custom|Stair
  (params (DamagedActor :type Actor%20Object%20Reference) (Damage :type Float%20%28single-precision%29) (InstigatedBy :type Controller%20Object%20Reference) (DamageCauser :type Actor%20Object%20Reference) (HitLocation :type Vector) (FHitComponent :type Primitive%20Component%20Object%20Reference) (BoneName :type Name) (Momentum :type Vector))
  (bind _returnvalue (Transformation|GetActorLocation DamagedActor))
  (bind _x (.x _returnvalue))
  (bind _y (.y _returnvalue))
  (bind _z (.z _returnvalue))
  (bind _returnvalue_2 (+ _z 1.9e+02))
  (bind _returnvalue_3 (Math|Vector|MakeVector _x _y _returnvalue_2))
  (bind _returnvalue_4 (Transformation|GetActorRightVector DamagedActor))
  (bind _returnvalue_5 (* _returnvalue_4 256.0))
  (bind _targettransform (Variables|Default|GetTargetTransform))
  (bind _location (.location _targettransform))
  (bind _returnvalue_6 (+ _returnvalue_5 _returnvalue_3))
  (Variables|Default|SetAttackingPlayer 0)
  (bind _output_get (Variables|Default|SetRightVector (Math|Vector|RotateVectorAroundAxis _returnvalue_4 52.0 (Transformation|GetActorForwardVector DamagedActor))))
  (Variables|Default|SetActor%20Location _returnvalue_6)
  (Development|PrintString "Spawned Actor")
  ;[1 1 1] if greater than. Use forward Vector, if less than, use inverse, then decide where to spawn the actor
  (if (> (Math|Float|Absolute(Float) (Math|Trig|Acos(Degrees) (Math|Vector|DotProduct _output_get (Math|Vector|GetUnitDirection(Vector) _returnvalue_6 (Transformation|GetActorLocation DamageCauser))))) 9e+01)
    (Development|PrintString "Backward Vector")
    (Variables|Default|SetTargetTransform (Math|Transform|MakeTransform (+ (Variables|Default|GetActor%20Location) (* (Variables|Default|GetRightVector) 128.0)) (Math|Conversions|RotationFromXVector (Variables|Default|GetRightVector))))
    (bind (_asynctaskproxy _targetdata _applicationtag) (Ability|Tasks|DoTargetSelection "None" "(TagName=\"Gameplay.Application.Explode\")" true)
      (:then
        (Ability|ExecuteGameplayCueWithParamsOnOwner (Variables|Default|GetGameplayCueTag) (Ability|GameplayCue|MakeGameplayCueParameters :NormalizedMagnitude 0.0 :RawMagnitude 0.0 :Location _location :Normal "0,0,0" :Instigator (Utilities|Casting|CastToFortPlayerPawn (Variables|Default|GetAttackingPlayer)) :GameplayEffectLevel 1 :AbilityLevel 1 :bReplicateLocationWhenUsingMinimalRepProxy false)))
      (:Targeted
        (Ability|ApplyGameplayEffectContainer _targetdata _applicationtag)))
    (else
      (Development|PrintString "Forward Vector")
      (Variables|Default|SetTargetTransform (Math|Transform|MakeTransform (- (Variables|Default|GetActor%20Location) (* (Variables|Default|GetRightVector) 128.0)) (Math|Conversions|RotationFromXVector (Math|Transform|InverseTransformDirection (Math|Conversions|ToTransform(Vector) (Variables|Default|GetRightVector)) (Variables|Default|GetRightVector)))))
      (bind (_asynctaskproxy_2 _targetdata_2 _applicationtag_2) (Ability|Tasks|DoTargetSelection "None" "(TagName=\"Gameplay.Application.Explode\")" true)
        (:then
          (Ability|ExecuteGameplayCueWithParamsOnOwner (Variables|Default|GetGameplayCueTag) (Ability|GameplayCue|MakeGameplayCueParameters :NormalizedMagnitude 0.0 :RawMagnitude 0.0 :Location _location :Normal "0,0,0" :Instigator (Utilities|Casting|CastToFortPlayerPawn (Variables|Default|GetAttackingPlayer)) :GameplayEffectLevel 1 :AbilityLevel 1 :bReplicateLocationWhenUsingMinimalRepProxy false)))
        (:Targeted
          (Ability|ApplyGameplayEffectContainer _targetdata_2 _applicationtag_2)))))
  (CallFunction|PlayClientEffects _location))

(event Custom|Cone/Floor
  (params (DamagedActor :type Actor%20Object%20Reference) (Damage :type Float%20%28single-precision%29) (InstigatedBy :type Controller%20Object%20Reference) (DamageCauser :type Actor%20Object%20Reference) (HitLocation :type Vector) (FHitComponent :type Primitive%20Component%20Object%20Reference) (BoneName :type Name) (Momentum :type Vector))
  (bind _returnvalue (Transformation|GetActorLocation DamagedActor))
  (bind _targettransform (Variables|Default|GetTargetTransform))
  (bind _location (.location _targettransform))
  (Variables|Default|SetAttackingPlayer 0)
  (bind _output_get (Variables|Default|SetUpVector (Transformation|GetActorUpVector DamagedActor)))
  (Variables|Default|SetRightVector (Transformation|GetActorRightVector DamagedActor))
  (Variables|Default|SetActor%20Location _returnvalue)
  (Development|PrintString "Spawned Actor")
  ;[1 1 1] if greater than. Use forward Vector, if less than, use inverse, then decide where to spawn the actor
  (if (> (Math|Float|Absolute(Float) (Math|Trig|Acos(Degrees) (Math|Vector|DotProduct _output_get (Math|Vector|GetUnitDirection(Vector) _returnvalue (Transformation|GetActorLocation DamageCauser))))) 9e+01)
    (Development|PrintString "Down Vector")
    (Variables|Default|SetTargetTransform (Math|Transform|MakeTransform (Math|Vector|MakeVector (+ (.x (+ (Variables|Default|GetActor%20Location) (* (Variables|Default|GetUpVector) 256.0))) (.x (* (Variables|Default|GetRightVector) 256.0))) (+ (.y (+ (Variables|Default|GetActor%20Location) (* (Variables|Default|GetUpVector) 256.0))) (.y (* (Variables|Default|GetRightVector) 256.0))) (+ (.z (+ (Variables|Default|GetActor%20Location) (* (Variables|Default|GetUpVector) 256.0))) 256.0)) (Math|Conversions|RotationFromXVector (Variables|Default|GetRightVector))))
    (bind (_asynctaskproxy _targetdata _applicationtag) (Ability|Tasks|DoTargetSelection "None" "(TagName=\"Gameplay.Application.Explode\")" true)
      (:then
        (Ability|ExecuteGameplayCueWithParamsOnOwner (Variables|Default|GetGameplayCueTag) (Ability|GameplayCue|MakeGameplayCueParameters :NormalizedMagnitude 0.0 :RawMagnitude 0.0 :Location _location :Normal "0,0,0" :Instigator (Utilities|Casting|CastToFortPlayerPawn (Variables|Default|GetAttackingPlayer)) :GameplayEffectLevel 1 :AbilityLevel 1 :bReplicateLocationWhenUsingMinimalRepProxy false)))
      (:Targeted
        (Ability|ApplyGameplayEffectContainer _targetdata _applicationtag)))
    (else
      (Development|PrintString "Up  Vector")
      (Variables|Default|SetTargetTransform (Math|Transform|MakeTransform (Math|Vector|MakeVector (+ (.x (- (Variables|Default|GetActor%20Location) (* (Variables|Default|GetUpVector) 256.0))) (.x (* (Variables|Default|GetRightVector) 256.0))) (+ (.y (- (Variables|Default|GetActor%20Location) (* (Variables|Default|GetUpVector) 256.0))) (.y (* (Variables|Default|GetRightVector) 256.0))) (- (.z (- (Variables|Default|GetActor%20Location) (* (Variables|Default|GetUpVector) 256.0))) 256.0)) (Math|Conversions|RotationFromXVector (Math|Transform|InverseTransformDirection (Math|Conversions|ToTransform(Vector) (Variables|Default|GetRightVector)) (Variables|Default|GetRightVector)))))
      (bind (_asynctaskproxy_2 _targetdata_2 _applicationtag_2) (Ability|Tasks|DoTargetSelection "None" "(TagName=\"Gameplay.Application.Explode\")" true)
        (:then
          (Ability|ExecuteGameplayCueWithParamsOnOwner (Variables|Default|GetGameplayCueTag) (Ability|GameplayCue|MakeGameplayCueParameters :NormalizedMagnitude 0.0 :RawMagnitude 0.0 :Location _location :Normal "0,0,0" :Instigator (Utilities|Casting|CastToFortPlayerPawn (Variables|Default|GetAttackingPlayer)) :GameplayEffectLevel 1 :AbilityLevel 1 :bReplicateLocationWhenUsingMinimalRepProxy false)))
        (:Targeted
          (Ability|ApplyGameplayEffectContainer _targetdata_2 _applicationtag_2)))))
  (CallFunction|PlayClientEffects _location))

(fn GetCustomAbilitySourceTransform
  (return (Variables|Default|GetTargetTransform)))
